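# Niagara effect for an artillery emote's firecracker spark, apparently a burning fuse on a skeletal mesh.
# Flame meshes, sparkle sprites and a glow sprite with light lerp across four sampled bone or socket positions (index 3 first) by system age, moving the spark along the mesh.
# A brief Flash sprite samples index 3.
# Flame camera offset is user-driven, and Glow alpha adjusts for front end versus gameplay.
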NiagaraSystem /BRCosmetics/Effects/Fort_Effects/Emotes/EssayViewPier/NS_Emote_Artillery_FirecrackerSpark
dependencies {
 usedDataInterfaces [
  &NiagaraDataInterfaceCurve
  &NiagaraDataInterfaceVectorCurve
  &NiagaraDataInterfaceColorCurve
  &NiagaraDataInterfaceSkeletalMesh
 ]
 usedDynamicInputs []
 usedModules [
  &self:SystemSpawnScript.SystemScriptSource.SystemScriptGraph.NiagaraNodeAssignment_1.SetVariables_#1
  &self:SystemSpawnScript.SystemScriptSource.SystemScriptGraph.NiagaraNodeAssignment_2.SetVariables_#2
  &SystemState
  &self:SystemSpawnScript.SystemScriptSource.SystemScriptGraph.NiagaraNodeAssignment_0.SetVariables_#3
 ]
}
emitters {
 Flame {
  emitterUpdateScript [
   &EmitterState()
   &SpawnRate(SpawnRate=30, Use Spawn Probability=false)
  ]
  particleSpawnScript [
   &InitializeParticle(
    Lifetime Max=0.25
    Lifetime Min=0.15
    Lifetime Mode=::NewEnumerator1
    Mesh Scale Max=<0.1 0.2 0.2>
    Mesh Scale Min=<0.02 0.05 0.05>
    Mesh Scale Mode=::NewEnumerator2
    Position Offset=<-1.5 0 0>
    UsePositionOffset=false
    Write Material Random=true
   ) {module=InitializeParticle001}
   /Game/Effects/Niagara/Modules/Orientation/FN_Fan(
    Cone Angle:NiagaraFloat=&UniformRangedFloat(Minimum=0, Maximum=0)
    Spin:NiagaraFloat=&UniformRangedFloat(Minimum=0, Maximum=6.2)
   )
   &InitialMeshOrientation(Mesh Orientation Mode=::NewEnumerator2, Rotation=<0 -0.23 0>, Use Rotation Vector=true) {module=InitialMeshOrientation001}
   &CameraOffset(Camera Offset Amount=$User.FireCameraOffset)
  ]
  particleUpdateScript [
   &SkeletalMeshLocation(
    Bone / Socket Index=3
    Bone Sampling Mode=::NewEnumerator4
    MaskSpawnGroup=false
    Position Sampling=::NewEnumerator2
    Skeletal Mesh=$System.Skeletal Mesh
   )
   &SkeletalMeshLocation(
    Bone / Socket Index=2
    Bone Sampling Mode=::NewEnumerator4
    MaskSpawnGroup=false
    Position Sampling=::NewEnumerator2
    Sampled Orientation Coordinate Space=::Local
    Skeletal Mesh=$System.Skeletal Mesh
   ) {module=SkeletalMeshLocation001}
   &SkeletalMeshLocation(
    Bone / Socket Index=1
    Bone Sampling Mode=::NewEnumerator4
    MaskSpawnGroup=false
    Position Sampling=::NewEnumerator2
    Sampled Orientation Coordinate Space=::Local
    Sampled Position Offset=<0 -1 0>
    Skeletal Mesh=$System.Skeletal Mesh
   ) {module=SkeletalMeshLocation002}
   &SkeletalMeshLocation(
    Bone Sampling Mode=::NewEnumerator4
    MaskSpawnGroup=false
    Position Sampling=::NewEnumerator2
    Sampled Orientation Coordinate Space=::Local
    Sampled Position Offset=<0 -2 0>
    Skeletal Mesh=$System.Skeletal Mesh
   ) {module=SkeletalMeshLocation003}
   /SetVariables(Particles.Position=&AddVectorToPosition(
     Position=/Niagara/DynamicInputs/Math/LerpMultipleVectors(
      Alpha:NiagaraFloat=$Engine.System.Age
      V 1:Vector3f=Linked() {linkedVariable {name=Particles.SkeletalMeshLocation.SampledPosition, type=NiagaraPosition}}
      V 2:Vector3f=Linked() {linkedVariable {name=Particles.SkeletalMeshLocation001.SampledPosition, type=NiagaraPosition}}
      V 3:Vector3f=Linked() {linkedVariable {name=Particles.SkeletalMeshLocation002.SampledPosition, type=NiagaraPosition}}
      V 4:Vector3f=Linked() {linkedVariable {name=Particles.SkeletalMeshLocation003.SampledPosition, type=NiagaraPosition}}
      V 5:Vector3f=?
      Vector Count:ENiagara_MultipleLerpCount=::NewEnumerator1
     )
     Vector=<0 0 0>
    )) {isSetParameters=true, module=SetVariables_#4}
   &ParticleState()
   &ScaleMeshSize(Scale Factor=&VectorFromCurve(
     VectorCurve=DataInterface() {propertyValues {
       XCurve {keys [<at 0 0.55>,<at 0.2 2.875>,<at 1 0>]}
       YCurve {keys [<at 0 0.05>,<at 0.2 2.5>,<at 1 0>]}
       ZCurve {keys [<at 0 0.55>,<at 0.2 2>,<at 1 0>]}
      }}
     Scale Curve=&Multiply_VectorByFloat(Vector=&Multiply_VectorByFloat(Vector=<1 1 1>, Float=$System.FireSizeScale), Float=0.53)
    ))
   &ScaleColor(
    Scale Alpha=0.5
    Scale RGB=&MakeVectorFromLinearColorRGB(Color=&ColorFromCurve(ColorCurve=DataInterface() {propertyValues {
        AlphaCurve {keys []}
        BlueCurve {keys [<at 0 0 RCIM_Cubic>,<at 1 0 RCIM_Cubic>]}
        GreenCurve {keys [<at 0 0.4167 RCIM_Cubic>,<at 1 0.5031 RCIM_Cubic>]}
        RedCurve {keys [<at 0 5 RCIM_Cubic>,<at 1 5 RCIM_Cubic>]}
       }}))
    ScaleA=true
    ScaleRGB=true
   )
  ]
  rendererClasses [&NiagaraMeshRendererProperties]
  renderers [{rendererClass=&NiagaraMeshRendererProperties, rendererIndex=0}]
 }
 Flash {
  emitterUpdateScript [
   &EmitterState(Inactive Response=::NewEnumerator1, Life Cycle Mode=::NewEnumerator1, Loop Behavior=::NewEnumerator1, Loop Duration=0.5, UseLoopDelay=false)
   &SpawnBurst_Instantaneous(Spawn Count=1, Use Spawn Probability=false)
  ]
  particleSpawnScript [
   &InitializeParticle(
    Color=<rgba 1 0.4615 0.0496 1>
    Lifetime=0.2
    Mesh Scale Mode=::NewEnumerator3
    Mesh Uniform Scale=0.015
    Position Offset=<0 0 5>
    Position Offset Coordinate Space=::Local
    Sprite Size Mode=::NewEnumerator3
    Uniform Sprite Size=50
    UsePositionOffset=true
    Write Material Random=true
   )
   &InitialMeshOrientation(Orientation Vector=<0 0 1>, Rotation=<0 0.5 0>, Use Orientation Vector=true, Use Rotation Vector=true)
   &CameraOffset(Camera Offset Amount=5)
  ]
  particleUpdateScript [
   &SkeletalMeshLocation(Bone / Socket Index=3, Bone Sampling Mode=::NewEnumerator4, MaskSpawnGroup=false, Skeletal Mesh=$System.Skeletal Mesh)
   &ParticleState()
   &ScaleSpriteSize(Uniform Curve Sprite Scale=DataInterface() {propertyValues {Curve {keys [{interpMode=RCIM_Cubic, tangentMode=RCTM_User, time=0, value=1}, {interpMode=RCIM_Cubic, tangentMode=RCTM_User, time=1, value=0}]}}})
   &DynamicMaterialParameters(
    Index 0 Param 1=1
    Index 0 Param 2=1
    Index 0 Param 3=1
    Index 0 Param 4=1
    Index 1 Param 1=1
    Index 1 Param 2=1
    Index 1 Param 3=1
    Index 1 Param 4=1
    Index 2 Param 2=HlslExpression() {hlslExpression=Emitter.Age*.2}
    Index 2 Param 3=2
    Index 2 Param 4=2
    Module.Param0WriteEnabled=true
    Module.Param0WriteEnabled1=true
    Module.Param0WriteEnabled2=true
    Module.Param1WriteEnabled=true
    Module.Param1WriteEnabled1=true
    Module.Param1WriteEnabled2=true
    Module.Param2WriteEnabled=true
    Module.Param2WriteEnabled1=true
    Module.Param2WriteEnabled2=true
    Module.Param3WriteEnabled=true
    Module.Param3WriteEnabled1=true
    Module.Param3WriteEnabled2=true
    Write Parameter Index 1=true
    Write Parameter Index 2=true
   ) {module=DynamicMaterialParameters001}
  ]
  rendererClasses [&NiagaraSpriteRendererProperties]
  renderers [{rendererClass=&NiagaraSpriteRendererProperties}]
 }
 Glow {
  emitterUpdateScript [
   &EmitterState(Inactive Response=::NewEnumerator1, Life Cycle Mode=::NewEnumerator1, Loop Behavior=::NewEnumerator1, Loop Duration=3, UseLoopDelay=false)
   &SpawnBurst_Instantaneous(Use Spawn Probability=false, UseLoopCountLimit=false)
  ]
  particleSpawnScript [
   &InitializeParticle(
    Color=<rgba 1 0.129 0.0499 0.5>
    Lifetime=5
    Position Offset=<0 0 5>
    Position Offset Coordinate Space=::Local
    Sprite Size Mode=::NewEnumerator3
    Uniform Sprite Size=33
    UsePositionOffset=true
    Write Material Random=true
   )
   &DynamicMaterialParameters(
    Index 0 Param 1=3
    Index 0 Param 2=3
    Module.Index0Param0Write=true
    Module.Index0Param1Write=true
    Module.Index0Param2Write=true
    Module.Index0Param3Write=true
   )
   /SetVariables(Particles.Position=&AddVectorToPosition(
     Position=$Particles.Position
     Vector=&TransformVector(Vector=$System.FXposition, Destination Space=::Simulation)
    )) {isSetParameters=true, module=SetVariables_#4}
  ]
  particleUpdateScript [
   &ParticleState(Kill Particles When Lifetime Has Elapsed=false, Let Infinitely Lived Particles Die When Emitter Deactivates=true)
   &FN_Light_Attributes(
    Color=<rgba 1 0.1249 0 1>
    Create Color Attribute=true
    Create Exponent Attribute=true
    Create Radius Attribute=true
    Exponent=3
    Radius=&RandomRangeFloat(Minimum=25, Maximum=40, Evaluation Type=::NewEnumerator1)
    Scale Color=10
   )
   &ScaleColor(Scale Alpha=&FN_FrontEndFloatModifier(Gameplay=1, Front End=0.5), ScaleA=true, ScaleRGB=true)
   &SkeletalMeshLocation(
    Bone / Socket Index=3
    Bone Sampling Mode=::NewEnumerator4
    MaskSpawnGroup=false
    Position Sampling=::NewEnumerator2
    Skeletal Mesh=$System.Skeletal Mesh
   )
   &SkeletalMeshLocation(
    Bone / Socket Index=2
    Bone Sampling Mode=::NewEnumerator4
    MaskSpawnGroup=false
    Position Sampling=::NewEnumerator2
    Sampled Orientation Coordinate Space=::Local
    Skeletal Mesh=$System.Skeletal Mesh
   ) {module=SkeletalMeshLocation001}
   &SkeletalMeshLocation(
    Bone / Socket Index=1
    Bone Sampling Mode=::NewEnumerator4
    MaskSpawnGroup=false
    Position Sampling=::NewEnumerator2
    Sampled Orientation Coordinate Space=::Local
    Skeletal Mesh=$System.Skeletal Mesh
   ) {module=SkeletalMeshLocation002}
   &SkeletalMeshLocation(
    Bone Sampling Mode=::NewEnumerator4
    MaskSpawnGroup=false
    Position Sampling=::NewEnumerator2
    Sampled Orientation Coordinate Space=::Local
    Skeletal Mesh=$System.Skeletal Mesh
   ) {module=SkeletalMeshLocation003}
   /SetVariables(Particles.Position=&AddVectorToPosition(
     Position=&LerpMultipleVectors(
      Alpha:NiagaraFloat=$Engine.System.Age
      V 1:Vector3f=Linked() {linkedVariable {name=Particles.SkeletalMeshLocation.SampledPosition, type=NiagaraPosition}}
      V 2:Vector3f=Linked() {linkedVariable {name=Particles.SkeletalMeshLocation001.SampledPosition, type=NiagaraPosition}}
      V 3:Vector3f=Linked() {linkedVariable {name=Particles.SkeletalMeshLocation002.SampledPosition, type=NiagaraPosition}}
      V 4:Vector3f=Linked() {linkedVariable {name=Particles.SkeletalMeshLocation003.SampledPosition, type=NiagaraPosition}}
      V 5:Vector3f=?
      Vector Count:ENiagara_MultipleLerpCount=::NewEnumerator1
     )
     Vector=<0 0 0>
    )) {isSetParameters=true}
  ]
  rendererClasses [&NiagaraSpriteRendererProperties, &NiagaraLightRendererProperties]
  renderers [
   {rendererClass=&NiagaraSpriteRendererProperties}
   {rendererClass=&NiagaraLightRendererProperties, rendererIndex=1}
  ]
 }
 Sparkles {
  emitterUpdateScript [
   &EmitterState()
   &SpawnRate(SpawnRate=25, Use Spawn Probability=true)
  ]
  particleSpawnScript [
   &InitializeParticle(
    Color=<rgba 20 4.7575 0 1>
    Lifetime=&RandomRangeFloat(Minimum=0.15, Maximum=0.3)
    Sprite Rotation Angle=&RandomRangeFloat(Minimum=-0.02, Maximum=0.02)
    Sprite Rotation Mode=::NewEnumerator1
    Sprite Size Mode=::NewEnumerator1
    Uniform Sprite Size Max=5
    Uniform Sprite Size Min=1
    UsePositionOffset=false
    Write Material Random=true
   )
   &AddVelocityInCone(
    Cone Angle=40
    Cone Axis=<0 0 1>
    Cone Axis Coordinate Space=::Simulation
    Module.Use Velocity Falloff On Cone Axis=false
    Velocity Strength=&RandomRangeFloat(Minimum=120, Maximum=110)
   )
  ]
  particleUpdateScript [
   &Drag(Drag=4, Use Linear Drag=true, Use Rotational Drag=true)
   &SolveForcesAndVelocity(Clamp Velocity=false, Limit Acceleration=false)
   &ParticleState()
   &ScaleSpriteSize(Scale Factor=&Vector2DFromFloat(Value=&FloatFromCurve(FloatCurve=DataInterface() {propertyValues {Curve {keys [<at 0 0 RCIM_Cubic>,<at 0.5 1 RCIM_Cubic>,<at 1 0 RCIM_Cubic>]}}}, Scale Curve=1.5)))
   &ScaleColor(
    Scale Alpha=&FloatFromCurve(FloatCurve=DataInterface() {propertyValues {Curve {keys [{arriveTangent=-1.402, interpMode=RCIM_Cubic, leaveTangent=-1.402, tangentMode=RCTM_User, time=0, value=1}, <at 0.13 1>, <at 1 0 RCIM_Cubic>]}}})
    ScaleA=true
    ScaleRGB=true
   )
   &CameraOffset(Camera Offset Amount=5)
   &SkeletalMeshLocation(
    Bone / Socket Index=3
    Bone Sampling Mode=::NewEnumerator4
    MaskSpawnGroup=false
    Position Sampling=::NewEnumerator2
    Skeletal Mesh=$System.Skeletal Mesh
   )
   &SkeletalMeshLocation(
    Bone / Socket Index=2
    Bone Sampling Mode=::NewEnumerator4
    MaskSpawnGroup=false
    Position Sampling=::NewEnumerator2
    Skeletal Mesh=$System.Skeletal Mesh
   ) {module=SkeletalMeshLocation001}
   &SkeletalMeshLocation(
    Bone / Socket Index=1
    Bone Sampling Mode=::NewEnumerator4
    MaskSpawnGroup=false
    Position Sampling=::NewEnumerator2
    Skeletal Mesh=$System.Skeletal Mesh
   ) {module=SkeletalMeshLocation002}
   &SkeletalMeshLocation(Bone Sampling Mode=::NewEnumerator4, MaskSpawnGroup=false, Position Sampling=::NewEnumerator2, Skeletal Mesh=$System.Skeletal Mesh) {module=SkeletalMeshLocation003}
   /SetVariables(Particles.Position=&AddVectorToPosition(
     Position=&LerpMultipleVectors(
      Alpha:NiagaraFloat=$Engine.System.Age
      V 1:Vector3f=Linked() {linkedVariable {name=Particles.SkeletalMeshLocation.SampledPosition, type=NiagaraPosition}}
      V 2:Vector3f=Linked() {linkedVariable {name=Particles.SkeletalMeshLocation001.SampledPosition, type=NiagaraPosition}}
      V 3:Vector3f=Linked() {linkedVariable {name=Particles.SkeletalMeshLocation002.SampledPosition, type=NiagaraPosition}}
      V 4:Vector3f=Linked() {linkedVariable {name=Particles.SkeletalMeshLocation003.SampledPosition, type=NiagaraPosition}}
      V 5:Vector3f=?
      Vector Count:ENiagara_MultipleLerpCount=::NewEnumerator1
     )
     Vector=&Lerp_Vector(A=<0 0 0>, B=&RandomRangeVector(Minimum=<-5 -5 10>, Maximum=<5 5 15>), Alpha=$Particles.NormalizedAge)
    )) {isSetParameters=true}
  ]
  rendererClasses [&NiagaraSpriteRendererProperties]
  renderers [{rendererClass=&NiagaraSpriteRendererProperties}]
 }
}
summary {
 systemName=NS_Emote_Artillery_FirecrackerSpark
 emitters [
  {emitterName=Sparkles, rendererClasses [&NiagaraSpriteRendererProperties]}
  {emitterName=Flame, rendererClasses [&NiagaraMeshRendererProperties]}
  {emitterName=Glow, rendererClasses [&NiagaraSpriteRendererProperties, &NiagaraLightRendererProperties]}
  {emitterName=Flash, rendererClasses [&NiagaraSpriteRendererProperties]}
 ]
 userVariables [
  User.bIsFrontEnd {type=NiagaraBool}
  User.bIsFrontEndPreview {type=NiagaraBool}
  User.FireCameraOffset {type=NiagaraFloat, defaultValue=15, description=""}
  User.MeshOffset {type=Vector3f, description="", defaultValue {struct=Vector3f, value=<0 0 0>}}
 ]
}
